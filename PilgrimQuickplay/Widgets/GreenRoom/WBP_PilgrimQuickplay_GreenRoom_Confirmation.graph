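# Green Room confirmation dialog for Pilgrim Quickplay, a CommonActivatableWidget that warns an uncalibrated player.
# It feeds an uncalibrated-player viewmodel into a UIKit dialog, focuses the first button, and toggles the player's Green Room mapping context.
# Button index 2 launches the FMCalibration modal from the player controller's component, then closes; index 3 only closes.
# Closing removes the confirmation scene through DynamicUIManager and deactivates the widget.
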WidgetBlueprint /PilgrimQuickplay/Widgets/GreenRoom/WBP_PilgrimQuickplay_GreenRoom_Confirmation
graphs [
 BP_GetDesiredFocusTarget {nodes [
   Get Desired Focus Target()
   n1 = Get Dialog_Base() -> (Dialog_Base:WBP UIKit Dialog Internal Object Reference)
   > Return Node!(
    execute
    ReturnValue:Widget Object Reference=GetChildAt!(
     self:Panel Widget Object Reference=Get ButtonsContainer!(self:WBP UIKit Dialog Internal Object Reference=@n1) -> (ButtonsContainer:Horizontal Box Object Reference)
     Index:int=0
    ) -> (ReturnValue:Widget Object Reference)
   )
  ]}
 EventGraph {nodes [
   n1 = CloseModal!() -> (OutputDelegate, then, ButtonIndex:int) {is=title K2Node_CustomEvent}
   > n2 = Switch on Int!(execute:Exec, Selection:int=@n1.ButtonIndex) -> (Default:Exec, 0:Exec, 1:Exec, 2:Exec, 3:Exec)
   n3 = Sequence(execute=@n2.2)
   > n4 = Cast To FortPlayerControllerAthena(Object=GetOwningPlayer())
   > Attempt to Launch Modal(self:BP FMCalibration Component Object Reference=GetComponentByClass!(
     self:Actor Object Reference=@n4
     ComponentClass:Actor Component Class Reference=/FMCalibration/Components/BP_FMCalibrationComponent.BP_FMCalibrationComponent_C
    ) -> (ReturnValue:BP FMCalibration Component Object Reference))
   RemoveScene!(
    execute=[@n3.then_1, @n2.3]
    self:Dynamic UIManager Object Reference=Get DynamicUIManager() -> (ReturnValue:Dynamic UIManager Object Reference) {is=title K2Node_GetSubsystem}
    Scene:Dynamic UIScene Object Reference=Get ConfirmationScene() -> (ConfirmationScene:Dynamic UIScene Object Reference)
    Player:"Player Controller Object Reference (by ref)"=GetFortPlayerControllerFromActor!(Actor:Actor Object Reference=GetOwningPlayer()) -> (ReturnValue:Fort Player Controller Object Reference)
   ) -> (then)
   > DeactivateWidget()
   Event PreConstruct()
   > n5 = Construct Object from Class(Class=&/DialogVM_UncalibratedPlayer.DialogVM_UncalibratedPlayer_C) -> (ReturnValue:Dialog VM Uncalibrated Player Object Reference)
   > SetDialogData!(execute, self:WBP UIKit Dialog Internal Object Reference=Get Dialog_Base(), Data:UI Kit Dialog Viewmodel Object Reference=@n5) -> (then)
   Event On Activated()
   > n6 = Construct Object from Class(Class=&DialogVM_UncalibratedPlayer_C)
   > n7 = SetDialogData(self=Get Dialog_Base(), Data=@n6)
   Bind Event to Dialog Button Clicked Dispatcher(execute=[@n9.then, @n7.then], self:WBP UIKit Dialog Internal Object Reference=Get Dialog_Base(), Delegate=@n1) {is=title K2Node_AddDelegate}
   n8 = GetComponentByClass(
    self=GetOwningPlayer()
    ComponentClass=/PilgrimCore/Player/Components/BP_PilgrimPlayerControllerComponent.BP_PilgrimPlayerControllerComponent_C
   ) -> (ReturnValue:BP Pilgrim Player Controller Component Object Reference)
   Is Valid(InputObject=@n8)
   > n9 = ToggleGreenroomMappingContext!(execute, self:BP Pilgrim Player Controller Component Object Reference=@n8, AddGreenroomMappingContext:bool=false) -> (then)
   n10 = GetComponentByClass(
    self=GetOwningPlayer()
    ComponentClass=&BP_PilgrimPlayerControllerComponent_C
   )
   Is Valid(InputObject=@n10)
   > ToggleGreenroomMappingContext(self=@n10, AddGreenroomMappingContext=true)
  ]}
]
variables [ConfirmationScene {type=DynamicUIScene, default=/PilgrimQuickplay/Widgets/Mobile/Pilgrim_Quickplay_GreenRoom_Confirmation}]
widgets {
 info {inheritedWidgetCount=0, parentClass=&CommonActivatableWidget, rootWidgetClass=&GridPanel, widgetCount=2}
 widgets [
  GridPanel_36 {widgetClassPath=&GridPanel, widget=GridPanel_36, properties {columnFill [1], rowFill [1]}}
  Dialog_Base {
   widgetClassPath=/FortUILibrary/Systems/Dialog/WBP_UIKit_Dialog_Internal
   bIsVariable=true
   parent=GridPanel_36
   widget=Dialog_Base
   properties {
    buttons {
     each Button Horizontal Fill=true
     each Button Min Width=230
     each Button Size=Medium
     focuse Button=0
     space Between Buttons=30
     cTA Buttons Types [Loud,Regular]
     group Placement {
      horizontal Alignment=HAlign_Left
      vertical Alignment=VAlign_Bottom
      grid {column=1, column Span=0, layer=0, row=2, row Span=0, nudge=<0 0>}
      padding=<ltrb 80 64 80 48>
     }
    }
    dialogSIze {
     height Override {enable=false, value=800}
     max Aspect Ratio {enable=false, value=0}
     max Desired Height {enable=true, value=800}
     max Desired Width {enable=false, value=0}
     min Aspect Ratio {enable=false, value=0}
     min Desired Height {enable=false, value=0}
     min Desired Width {enable=false, value=0}
     width Override {enable=false, value=1200}
    }
   }
   slot {class=GridSlot}
  }
 ]
}
