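# FM Core button widget derived from WBP_UIKit_Button_Regular that restyles it with FMCore text and background materials, adding no widgets of its own.
# PreConstruct, Construct and On Widget Styled each call the parent, then run one shared styling chain.
# SetFontMaterial rebuilds the large, medium and small primary fonts with the Font Material variable and stencil disabled.
# UpdateFonts applies them, then MI_FMCore_Button_Regular becomes the background material via UpdateBackground.
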WidgetBlueprint /SparksUICommon/UILibrary/WBP_FM_Core_Button_SmallerRadius
graphs [
 EventGraph {nodes [
   n1 = Event PreConstruct()
   > n2 = Parent: PreConstruct(IsDesignTime=@n1.IsDesignTime)
   n3 = SetFontMaterial!(
    execute=[@n2.then, @n6.then, @n7.then]
    self
    Font Info:"Slate Font Info Structure (by ref)"=Get LargePrimaryFont() -> (LargePrimaryFont:Slate Font Info Structure)
   ) -> (then, SlateFontInfo:Slate Font Info Structure)
   > Set LargePrimaryFont(LargePrimaryFont:Slate Font Info Structure=@n3) -> (Output_Get:Slate Font Info Structure)
   > n4 = SetFontMaterial(Font Info=Get MediumPrimaryFont() -> (MediumPrimaryFont:Slate Font Info Structure))
   > Set MediumPrimaryFont(MediumPrimaryFont:Slate Font Info Structure=@n4) -> (Output_Get:Slate Font Info Structure)
   > n5 = SetFontMaterial(Font Info=Get SmallPrimaryFont() -> (SmallPrimaryFont:Slate Font Info Structure))
   > Set SmallPrimaryFont(SmallPrimaryFont:Slate Font Info Structure=@n5) -> (Output_Get:Slate Font Info Structure)
   > UpdateFonts(self:WBP UIKit Button CTA Base Object Reference)
   > Set BackgroundMaterial(BackgroundMaterial:Material Instance Object Reference=&/MI_FMCore_Button_Regular) -> (Output_Get:Material Instance Object Reference)
   > UpdateBackground(self:WBP UIKit Button CTA Base Object Reference)
   Event Construct()
   > n6 = Parent: Construct()
   Event On Widget Styled() {is=title K2Node_Event}
   > n7 = Parent: On Widget Styled() {is=title K2Node_CallParentFunction}
   Event Tick()
  ]}
 SetFontMaterial {nodes [
   n1 = SetFontMaterial!() -> (then, Font Info:"Slate Font Info Structure (by ref)") {is=title K2Node_FunctionEntry}
   n2 = Break Slate Font Info!(SlateFontInfo:"Slate Font Info Structure (by ref)"=@n1) -> (
    FontObject:Object Reference
    FontMaterial:Object Reference
    OutlineSettings:Font Outline Settings Structure
    TypefaceFontName:Name
    Size:float=~
    LetterSpacing:int
    SkewAmount:float=~
    bForceMonospaced:bool=~
    bMaterialIsStencil:bool=~
    MonospacedWidth:float=~
   ) {is=title K2Node_BreakStruct}
   > Return Node!(
    execute
    SlateFontInfo:Slate Font Info Structure=MakeSlateFontInfo!(
     FontObject:Object Reference=@n2
     FontMaterial:Object Reference=Get Font Material() -> (Font Material:Object Reference)
     OutlineSettings:Font Outline Settings Structure=@n2.OutlineSettings
     TypefaceFontName:Name=@n2.TypefaceFontName
     Size:float=@n2.Size
     LetterSpacing:int=@n2.LetterSpacing
     SkewAmount:float=@n2.SkewAmount
     bForceMonospaced:bool=@n2.bForceMonospaced
     bMaterialIsStencil:bool=false
     MonospacedWidth:float=@n2.MonospacedWidth
    ) -> (ReturnValue:Slate Font Info Structure)
   )
  ]}
]
variables [Font Material {type=Object, default=&/MI_FMCore_Button_Text}]
widgets {
 info {
  inheritedWidgetCount=0
  parentClass=/FortUILibrary/UIKit/Buttons/WBP_UIKit_Button_Regular.WBP_UIKit_Button_Regular_C
  rootWidgetClass=null
  widgetCount=0
 }
 widgets []
}
