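# More Options menu screen for Festival quickplay (Pilgrim), with Tune-up, Controls, Tips, Settings and Exit buttons.
# Buttons open calibration, control-scheme, tutorial or settings screens; Exit, Tune-up and Tips broadcast ShowSparksQuestScreen false on the playspace event router.
# Calibration modal close events rebroadcast that show/hide state, and Controls follows input type, collapsing on touch.
# It caches the player's music battle via a stateful event, but Update Exit Button Status has an unwired entry.
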WidgetBlueprint /PilgrimQuickplay/Widgets/GreenRoom/WBP_PilgrimQuickplay_MoreOptionsMenu
graphs [
 BP_GetDesiredFocusTarget {nodes [
   Get Desired Focus Target()
   > Return Node!(execute, ReturnValue:Widget Object Reference=Get Tune-up() -> (Tune-up:WBP FM Core Button Smaller Radius Object Reference))
  ]}
 Close Options Menu {nodes [
   Close Options Menu() {is=title K2Node_FunctionEntry}
   > n1 = GetEventRouter!(execute, Scope:Object Class Reference=/Script/FortniteGame.FortPlayspace, ContextActor:Actor Object Reference=GetOwningPlayer()) -> (then, ReturnValue:Gameplay Event Router Component Object Reference)
   > Broadcast Event!(
    execute
    self:Gameplay Event Router Component Object Reference=@n1
    EventData:Show Sparks Quest Screen Event Structure=Make Show Sparks Quest Screen Event!(bShouldShow:bool=False) -> (ShowSparksQuestScreenEvent:Show Sparks Quest Screen Event Structure) {is=title K2Node_MakeStruct}
    EventContext:Object Reference
   ) -> (then, ReturnValue:bool)
  ]}
 ConfirmRestart {nodes [
   ConfirmRestart() {is=title K2Node_FunctionEntry}
   Return Node!(execute, ScreenSwitcher:Common Visibility Switcher Object Reference)
  ]}
 EventGraph {nodes [
   "On Clicked (BackButton)"!() -> (OutputDelegate, then, Button:Common Button Base Object Reference) {is=title K2Node_ComponentBoundEvent}
   > Close Options Menu() {is=title K2Node_CallFunction}
   "On Clicked (Settings)"!() -> (OutputDelegate, then, Button:Common Button Base Object Reference) {is=title K2Node_ComponentBoundEvent}
   n1 = GetFortPlayerControllerFromActor!(Actor:Actor Object Reference=GetOwningPlayer()) -> (ReturnValue:Fort Player Controller Object Reference)
   > n2 = Create Widget!(
    execute:Exec
    Class:User Widget Class Reference=/FortniteCoreUI/Settings/SettingScreen.SettingScreen_C
    OwningPlayer:Player Controller Object Reference=@n1
   ) -> (then:Exec, ReturnValue:Setting Screen Object Reference)
   > PushContentWidgetForPlayer!(execute, LocalPlayerController:Fort Player Controller Object Reference=@n1, PushWidget:Widget Object Reference=@n2) -> (then)
   > NavigateToSetting!(
    execute
    self:Fort Settings Screen Object Reference=@n2
    SettingCategory:Name=Get SettingCategoryGameplay() -> (SettingCategoryGameplay:Name=None)
    SettingToSelect:String=Get SettingToSelectFestival() -> (SettingToSelectFestival:String)
   ) -> (then)
   "On Clicked (Tips)"!() -> (OutputDelegate, then, Button:Common Button Base Object Reference) {is=title K2Node_ComponentBoundEvent}
   n3 = Get DynamicUIManager() -> (ReturnValue:Dynamic UIManager Object Reference) {is=title K2Node_GetSubsystem}
   > Is Valid(InputObject=@n3)
   > AddScene!(
    execute
    self:Dynamic UIManager Object Reference=@n3
    Scene:Dynamic UIScene Object Reference=Get TipsScene() -> (TipsScene:Dynamic UIScene Object Reference)
    Player:"Player Controller Object Reference (by ref)"=GetOwningPlayer()
   ) -> (then)
   > n4 = GetEventRouter(Scope=&FortPlayspace, ContextActor=GetOwningPlayer())
   > Broadcast Event(self=@n4, EventData=Make Show Sparks Quest Screen Event(bShouldShow=False))
   "On Clicked (Controls)"!() -> (OutputDelegate, then, Button:Common Button Base Object Reference) {is=title K2Node_ComponentBoundEvent}
   n5 = GetOwningPlayer()
   > Branch(Condition=NOT Boolean(A=IsUsingTouch(PlayerController=@n5)))
   n6 = Get DynamicUIManager()
   > Is Valid(InputObject=@n6)
   > AddScene(self=@n6, Scene=/PilgrimCore/Widgets/Pilgrim_ControlSchemes, Player=GetOwningPlayer())
   "On Clicked (Tune-up)"!() -> (OutputDelegate, then, Button:Common Button Base Object Reference) {is=title K2Node_ComponentBoundEvent}
   > Broadcast Event(
    self=GetGlobalEventRouter()
    EventData:FMEvent Calibration Toggle Modal Structure=Make FMEvent Calibration Toggle Modal!() -> (FMEvent_Calibration_ToggleModal:FMEvent Calibration Toggle Modal Structure) {is=title K2Node_MakeStruct}
   )
   > n7 = GetEventRouter(Scope=&FortPlayspace, ContextActor=GetOwningPlayer())
   > Broadcast Event(self=@n7, EventData:Show Sparks Quest Screen Event Structure=Make Show Sparks Quest Screen Event(bShouldShow=False))
   n8 = OnPlayspaceUserChanged!() -> (OutputDelegate, then, AddedUser:"Playspace User Structure (by ref)") {is=title K2Node_CustomEvent}
   > Update All Button Status()
   n9 = RefreshControlsIcon!() -> (OutputDelegate, then, bNewInputType:ECommonInputType Enum) {is=title K2Node_CustomEvent}
   n10 = Get ControllerIcon() -> (ControllerIcon:Texture 2D Object Reference)
   > SetIcon!(
    execute
    self:WBP UIKit Button CTA Base Object Reference=Get Controls() -> (Controls:WBP FM Core Button Smaller Radius Object Reference)
    Icon:Object Reference=Select!(
     MouseAndKeyboard:Object Reference=Get MouseAndKeyboardIcon() -> (MouseAndKeyboardIcon:Texture 2D Object Reference)
     Gamepad:Object Reference=@n10
     Touch:Object Reference=@n10
     Count:Object Reference=@n10
     Index:ECommonInputType Enum=@n9.bNewInputType
    ) -> (ReturnValue:Object Reference)
   ) -> (then)
   Event Construct()
   > Update All Button Status()
   n11 = Get CommonInputSubsystem!(PlayerController:Player Controller Object Reference=GetOwningPlayer()) -> (ReturnValue:Common Input Object Reference) {is=title K2Node_GetSubsystemFromPC}
   > Bind Event to On Input Method Changed(self=@n11, Delegate=Create Event())
   > RefreshControlsIcon!(execute, self, bNewInputType:ECommonInputType Enum=GetCurrentInputType(self=@n11)) -> (then) {is=title K2Node_CallFunction}
   > n12 = StartListeningToEvent!(
    execute:Exec
    Target:Gameplay Event Router Component Object Reference=GetGlobalEventRouter()
    EventType:Script Struct Object Reference=/Script/FMCalibrationRuntime.FMEvent_Calibration_ModalClosed
    Context:Object Reference
    EventBubblingRule:EEventBubblingRule Enum
   ) -> (
    then:Exec
    AsyncTaskProxy:Async Action Start Listening To Event Object Reference
    OnEventReceived:Exec
    Payload:FMEvent Calibration Modal Closed Structure
   )
   > Set CalibrationModalClosed Task Ref(CalibrationModalClosed Task Ref:Async Action Start Listening To Event Object Reference=@n12) -> (Output_Get:Async Action Start Listening To Event Object Reference)
   > n13 = StartListeningToEvent(Target=GetGlobalEventRouter(), EventType=/Script/FMCalibrationRuntime.FMEvent_Calibration_ModalCloseToQuest) -> (Payload:FMEvent Calibration Modal Close To Quest Structure)
   > Set CalibrationModalCloseToQuest Task(CalibrationModalCloseToQuest Task:Async Action Start Listening To Event Object Reference=@n13) -> (Output_Get:Async Action Start Listening To Event Object Reference)
   Show Options Menu(execute=@n13.OnEventReceived)
   Close Options Menu(execute=@n12.OnEventReceived)
   Event On Activated()
   > Update All Button Status()
   Event Destruct()
   > Parent: Destruct()
   > n14 = Sequence!(execute) -> (then_0, then_1, then_2:Exec, then_3:Exec)
   n15 = Get PilgrimMBIStartedEventHandle() -> (PilgrimMBIStartedEventHandle:Async Action Start Listening To Stateful Event Object Reference)
   > Is Valid(InputObject=@n15)
   > StopListeningToStatefulEvent(self:Async Action Start Listening To Stateful Event Object Reference=@n15)
   > Set PilgrimMBIStartedEventHandle(PilgrimMBIStartedEventHandle:Async Action Start Listening To Stateful Event Object Reference) -> (Output_Get:Async Action Start Listening To Stateful Event Object Reference)
   n16 = Get SparksMusicPlayspace() -> (SparksMusicPlayspace:Sparks Music Playspace Object Reference)
   Is Valid(exec=@n14.then_1, InputObject=@n16)
   n17 = Create Event()
   > Unbind Event from Notify Playspace User Added(self:Playspace Object Reference=@n16, Delegate=@n17) {is=title K2Node_RemoveDelegate}
   > Unbind Event from Notify Playspace User Removed(self:Playspace Object Reference=@n16, Delegate=@n17) {is=title K2Node_RemoveDelegate}
   Unbind Event from On Input Method Changed(execute=@n14.then_2, self=Get CommonInputSubsystem(PlayerController=GetOwningPlayer()), Delegate=Create Event())
   StopListeningToEvent(
    execute=@n14.then_3
    self:Async Action Start Listening To Event Object Reference=Get CalibrationModalClosed Task Ref() -> (CalibrationModalClosed Task Ref:Async Action Start Listening To Event Object Reference)
   )
   > StopListeningToEvent(self=Get CalibrationModalCloseToQuest Task() -> (CalibrationModalCloseToQuest Task:Async Action Start Listening To Event Object Reference))
   Event OnInitialized()
   > n18 = PilgrimWaitForEventRouter!(execute:Exec, ContextActor:Actor Object Reference=GetOwningPlayer()) -> (then:Exec, OnEventRouterAvailable:Exec, EventRouter:Gameplay Event Router Component Object Reference) {is=title K2Node_AsyncAction}
   n19 = GetSparksMusicPlayspaceOfClass!(
    PlayspaceType:Sparks Music Playspace Class Reference=/Script/SparksMusicPlayspaceRuntime.SparksMusicPlayspace
    Actor:Actor Object Reference=GetOwningPlayer()
    bMatchChildTypes:bool=false
   ) -> (ReturnValue:Sparks Music Playspace Object Reference)
   > Is Valid(InputObject=@n19)
   > n20 = Set SparksMusicPlayspace(SparksMusicPlayspace:Sparks Music Playspace Object Reference=@n19) -> (Output_Get:Sparks Music Playspace Object Reference)
   > Bind Event to Notify Playspace User Added(self:Playspace Object Reference=@n20, Delegate=@n8) {is=title K2Node_AddDelegate}
   > Bind Event to Notify Playspace User Removed(self:Playspace Object Reference=@n20, Delegate=@n8) {is=title K2Node_AddDelegate}
   n21 = StartListeningToStatefulEvent!(
    execute:Exec=@n18.OnEventRouterAvailable
    Target:Gameplay Event Router Component Object Reference=@n18
    EventType:Script Struct Object Reference=/Script/PilgrimCoreRuntime.PilgrimEvent_MBI_Started
    Context:Object Reference
    EventBubblingRule:EEventBubblingRule Enum
    None:%=@n18
   ) -> (
    then:Exec
    AsyncTaskProxy:Async Action Start Listening To Stateful Event Object Reference
    OnEventReceived:Exec
    OnSavedEventStateExists:Exec
    OnEventStateCleared:Exec
    Payload:Pilgrim Event MBI Started Structure
   ) {is=title K2Node_StartListeningToStatefulEventAsyncAction}
   > Set PilgrimMBIStartedEventHandle(PilgrimMBIStartedEventHandle=@n21)
   n22 = Cast To MusicBattleInstance!(
    execute=[@n21.OnEventReceived, @n21.OnSavedEventStateExists]
    Object=Break Pilgrim Event MBI Event Base!(PilgrimEvent_MBI_EventBase:"Pilgrim Event MBI Event Base Structure (by ref)"=@n21.Payload) -> (MusicBattleInstance:Pilgrim Music Battle Instance Weak object) {is=title K2Node_BreakStruct}
   ) -> (then, CastFailed, AsMusic Battle Instance:Music Battle Instance Object Reference) {is=title K2Node_DynamicCast}
   > n23 = Cast To FortPlayerStateAthena!(execute, Object=Get PlayerState!(self:Controller Object Reference=GetOwningPlayer()) -> (PlayerState:Player State Object Reference)) -> (then, CastFailed, AsFort Player State Athena:Fort Player State Athena Object Reference) {is=title K2Node_DynamicCast}
   > Branch(Condition=IsPlayerInBattleParticipants!(self:Pilgrim Music Battle Instance Object Reference=@n22, PlayerState:Fort Player State Athena Object Reference=@n23) -> (ReturnValue:bool))
   > Set MusicBattleInstance(MusicBattleInstance:Music Battle Instance Object Reference=@n22) -> (Output_Get:Music Battle Instance Object Reference)
   Event On Deactivated()
  ]}
 Exit to Backstage {nodes [
   Exit to Backstage() {is=title K2Node_FunctionEntry}
   n1 = GetComponentByClass!(
    self:Actor Object Reference=GetOwningPlayer()
    ComponentClass:Actor Component Class Reference=/PilgrimCore/Player/Components/BP_PilgrimPlayerControllerComponent.BP_PilgrimPlayerControllerComponent_C
   ) -> (ReturnValue:BP Pilgrim Player Controller Component Object Reference)
   > Is Valid(InputObject=@n1)
   > SendRequestToStopAndSkipPostGame(self:Pilgrim Player Controller Component Object Reference=@n1)
  ]}
 Show Options Menu {nodes [
   Show Options Menu() {is=title K2Node_FunctionEntry}
   > n1 = GetEventRouter(Scope=&FortPlayspace, ContextActor=GetOwningPlayer())
   > Broadcast Event(self=@n1, EventData=Make Show Sparks Quest Screen Event(bShouldShow=true))
  ]}
 Update All Button Status {nodes [
   Update All Button Status() {is=title K2Node_FunctionEntry}
   n1 = GetOwningPlayer()
   > SetVisibility(
    self=Get Controls()
    InVisibility=Select!(Option 0:ESlateVisibility Enum=Visible, Option 1:ESlateVisibility Enum=Collapsed, Index:bool=IsUsingTouch(PlayerController=@n1)) -> (ReturnValue:ESlateVisibility Enum)
   )
   > Update Exit Button Status()
  ]}
 Update Exit Button Status {nodes [
   n1 = Get MusicBattleInstance() -> (MusicBattleInstance:Music Battle Instance Object Reference)
   n2 = Is Valid(InputObject=@n1)
   SetIsEnabled(execute=@n2.Is Not Valid, bInIsEnabled=false)
   SetIsEnabled(execute=@n2.Is Valid, bInIsEnabled=IsHardStopAllowed!(self:Pilgrim Music Battle Instance Object Reference=@n1) -> (ReturnValue:bool))
   SetIsEnabled(bInIsEnabled=true)
   Update Exit Button Status() {is=title K2Node_FunctionEntry}
  ]}
]
variables [
 Active Index {type=integer, default=-1}
 CalibrationModalClosed Task Ref {type=AsyncAction_StartListeningToEvent, default=null}
 CalibrationModalCloseToQuest Task {type=AsyncAction_StartListeningToEvent, default=null}
 ControllerIcon {type=Texture2D, default=/Game/UI/Icons/Icon-Settings/T-Icon-Controller-64}
 MouseAndKeyboardIcon {type=Texture2D, default=/Game/UI/Icons/Icon-Settings/T-Icon-Keyboard-64}
 MusicBattleInstance {type=MusicBattleInstance_C, default=null}
 PilgrimMBIStartedEventHandle {type=AsyncAction_StartListeningToStatefulEvent, default=null}
 SettingCategoryGameplay {type=string, default=GameplayCollection}
 SettingToSelectFestival {type=string, default=Calibration Profile}
 SparksMusicPlayspace {type=SparksMusicPlayspace, default=null}
 TipsScene {type=DynamicUIScene, default=/PilgrimQuickplay/Widgets/Scenes/UIScene_PilgrimQuickplay_Tutorials}
]
widgets {
 info {inheritedWidgetCount=0, parentClass=/Script/FortniteUI.AthenaMapScreenTabBase, rootWidgetClass=&Overlay, widgetCount=20}
 widgets [
  Overlay_47 {widgetClassPath=&Overlay, widget=Overlay_47}
  Background {
   widgetClassPath=&Image
   bIsVariable=true
   parent=Overlay_47
   widget=Background
   properties {brush {
     bIsDynamicallyLoaded=false
     drawAs=Image
     imageType=FullColor
     mirroring=NoMirror
     resourceName=None
     resourceObject=/FMCalibration/Materials/MI_UI_FTUE_Sparks_Fullscreen_Background
     tiling=NoTile
     imageSize=<1000 1000>
     margin=<ltrb 0 0 0 0>
     outlineSettings {
      bUseBrushTransparency=false
      roundingType=HalfHeightRadius
      width=0
      color {colorUseRule=UseColor_Specified, specifiedColor=<rgba 0 0 0 0>}
      cornerRadii=<0 0 0 0>
     }
     tintColor {colorUseRule=UseColor_Specified, specifiedColor=<rgba 1 1 1 1>}
     uVRegion {bIsValid=false, max=<0 0>, min=<0 0>}
    }}
   slot {class=OverlaySlot, properties {horizontalAlignment=HAlign_Fill, verticalAlignment=VAlign_Fill}}
  }
  Dust {
   widgetClassPath=&Image
   bIsVariable=true
   parent=Overlay_47
   widget=Dust
   properties {
    visibility=Collapsed
    brush {
     bIsDynamicallyLoaded=false
     drawAs=Image
     imageType=FullColor
     mirroring=NoMirror
     resourceName=None
     resourceObject=/FMCalibration/Materials/MI_Ui_FTUE_RadiatingAtmospherics
     tiling=NoTile
     imageSize=<1200 1200>
     margin=<ltrb 0 0 0 0>
     outlineSettings {
      bUseBrushTransparency=false
      roundingType=HalfHeightRadius
      width=0
      color {colorUseRule=UseColor_Specified, specifiedColor=<rgba 0 0 0 0>}
      cornerRadii=<0 0 0 0>
     }
     tintColor {colorUseRule=UseColor_Specified, specifiedColor=<rgba 1 1 1 1>}
     uVRegion {bIsValid=false, max=<0 0>, min=<0 0>}
    }
    colorAndOpacity=<rgba 1 1 1 0.6>
   }
   slot {class=OverlaySlot, properties {horizontalAlignment=HAlign_Fill, verticalAlignment=VAlign_Fill}}
  }
  Glow {
   widgetClassPath=&Image
   bIsVariable=true
   parent=Overlay_47
   widget=Glow
   properties {
    brush {
     bIsDynamicallyLoaded=false
     drawAs=Image
     imageType=FullColor
     mirroring=NoMirror
     resourceName=None
     resourceObject=/FMCalibration/Materials/MI_UI_FTUE_AlphaComposite_Glow_Blue_Bottom
     tiling=NoTile
     imageSize=<1200 1200>
     margin=<ltrb 0 0 0 0>
     outlineSettings {
      bUseBrushTransparency=false
      roundingType=HalfHeightRadius
      width=0
      color {colorUseRule=UseColor_Specified, specifiedColor=<rgba 0 0 0 0>}
      cornerRadii=<0 0 0 0>
     }
     tintColor {colorUseRule=UseColor_Specified, specifiedColor=<rgba 1 1 1 1>}
     uVRegion {bIsValid=false, max=<0 0>, min=<0 0>}
    }
    colorAndOpacity=<rgba 1 1 1 0.6>
   }
   slot {class=OverlaySlot, properties {horizontalAlignment=HAlign_Fill, verticalAlignment=VAlign_Fill}}
  }
  Glow_Top {
   widgetClassPath=&Image
   bIsVariable=true
   parent=Overlay_47
   widget=Glow_Top
   properties {brush {
     bIsDynamicallyLoaded=false
     drawAs=Image
     imageType=FullColor
     mirroring=NoMirror
     resourceName=None
     resourceObject=/FMCalibration/Materials/MI_UI_FTUE_AlphaComposite_Glow1_Blue_Top
     tiling=NoTile
     imageSize=<1000 1000>
     margin=<ltrb 0 0 0 0>
     outlineSettings {
      bUseBrushTransparency=false
      roundingType=HalfHeightRadius
      width=0
      color {colorUseRule=UseColor_Specified, specifiedColor=<rgba 0 0 0 0>}
      cornerRadii=<0 0 0 0>
     }
     tintColor {colorUseRule=UseColor_Specified, specifiedColor=<rgba 1 1 1 1>}
     uVRegion {bIsValid=false, max=<0 0>, min=<0 0>}
    }}
   slot {class=OverlaySlot, properties {horizontalAlignment=HAlign_Center, verticalAlignment=VAlign_Fill}}
  }
  SafeZone_BottomNav {
   widgetClassPath=&SafeZone
   bIsVariable=true
   parent=Overlay_47
   widget=SafeZone_BottomNav
   properties {padLeft=false, padTop=false}
   slot {class=OverlaySlot, properties {horizontalAlignment=HAlign_Right, verticalAlignment=VAlign_Bottom, padding=<ltrb 48 48 48 48>}}
  }
  Overlay_0 {widgetClassPath=&Overlay, parent=SafeZone_BottomNav, widget=Overlay_0, slot {class=SafeZoneSlot}}
  WBP_CaptureForPostBufferUpdate {
   widgetClassPath=&WBP_CaptureForPostBufferUpdate
   bIsVariable=true
   parent=Overlay_0
   slot {class=OverlaySlot, properties {horizontalAlignment=HAlign_Fill, verticalAlignment=VAlign_Fill}}
  }
  BackButton {
   widgetClassPath=/FortUILibrary/UIKit/Buttons/WBP_UIKit_Button_Quiet
   bIsVariable=true
   parent=Overlay_0
   widget=BackButton
   properties {text=Exit, triggeringInputAction {dataTable=&InputActionDataTable, rowName=GenericBack}}
   slot {class=OverlaySlot}
  }
  StageTech_ImageBox {
   widgetClassPath=&Overlay
   bIsVariable=true
   parent=Overlay_47
   widget=StageTech_ImageBox
   slot {class=OverlaySlot, properties {verticalAlignment=VAlign_Bottom, padding=<ltrb -120 0 0 0>}}
  }
  Glow_Box {
   widgetClassPath=&Overlay
   parent=StageTech_ImageBox
   widget=Glow_Box
   slot {class=OverlaySlot, properties {horizontalAlignment=HAlign_Center, verticalAlignment=VAlign_Center}}
  }
  PinkGlow {
   widgetClassPath=&Image
   bIsVariable=true
   parent=Glow_Box
   widget=PinkGlow
   properties {
    brush {
     bIsDynamicallyLoaded=false
     drawAs=Image
     imageType=FullColor
     mirroring=NoMirror
     resourceName=None
     resourceObject=/FMCalibration/Materials/MI_UI_GlowTechPink
     tiling=NoTile
     imageSize=<900 900>
     margin=<ltrb 0 0 0 0>
     outlineSettings {
      bUseBrushTransparency=false
      roundingType=HalfHeightRadius
      width=0
      color {colorUseRule=UseColor_Specified, specifiedColor=<rgba 0 0 0 0>}
      cornerRadii=<0 0 0 0>
     }
     tintColor {colorUseRule=UseColor_Specified, specifiedColor=<rgba 1 1 1 0.2>}
     uVRegion {bIsValid=false, max=<0 0>, min=<0 0>}
    }
    renderTransform {angle=0, scale=<1 1>, shear=<0 0>, translation=<-50 -55>}
   }
   slot {class=OverlaySlot, properties {padding=<ltrb 150 0 0 0>}}
  }
  BlueGlow {
   widgetClassPath=&Image
   bIsVariable=true
   parent=Glow_Box
   widget=BlueGlow
   properties {
    brush {
     bIsDynamicallyLoaded=false
     drawAs=Image
     imageType=FullColor
     mirroring=NoMirror
     resourceName=None
     resourceObject=/FMCalibration/Materials/MI_UI_GlowTechBlue
     tiling=NoTile
     imageSize=<600 600>
     margin=<ltrb 0 0 0 0>
     outlineSettings {
      bUseBrushTransparency=false
      roundingType=HalfHeightRadius
      width=0
      color {colorUseRule=UseColor_Specified, specifiedColor=<rgba 0 0 0 0>}
      cornerRadii=<0 0 0 0>
     }
     tintColor {colorUseRule=UseColor_Specified, specifiedColor=<rgba 1 1 1 0.2>}
     uVRegion {bIsValid=false, max=<0 0>, min=<0 0>}
    }
    renderTransform {angle=0, scale=<1 1>, shear=<0 0>, translation=<180 350>}
   }
   slot {class=OverlaySlot}
  }
  WBP_FMCalibration_StageTech {
   widgetClassPath=/FMCalibration/Widgets/WBP_FMCalibration_StageTech
   bIsVariable=true
   parent=StageTech_ImageBox
   widget=WBP_FMCalibration_StageTech
   properties {renderTransform {angle=0, scale=<1.1 1.1>, shear=<0 0>, translation=<0 0>}}
   slot {class=OverlaySlot, properties {verticalAlignment=VAlign_Bottom, padding=<ltrb -50 0 0 45>}}
  }
  SizeBox_0 {
   widgetClassPath=&SizeBox
   parent=Overlay_47
   widget=SizeBox_0
   properties {bOverride_WidthOverride=true, widthOverride=408}
   slot {class=OverlaySlot, properties {horizontalAlignment=HAlign_Center, verticalAlignment=VAlign_Center}}
  }
  SongOptionButtons {widgetClassPath=&VerticalBox, parent=SizeBox_0, widget=SongOptionButtons, slot {class=SizeBoxSlot}}
  Tune-up {
   widgetClassPath=/SparksUICommon/UILibrary/WBP_FM_Core_Button_SmallerRadius
   bIsVariable=true
   parent=SongOptionButtons
   widget=Tune-up
   properties {icon=/FMCore/Textures/T_Icon_GlobalControls, text=Tune-up}
   slot {class=VerticalBoxSlot, properties {padding=<ltrb 10 10 10 10>}}
  }
  Controls {
   widgetClassPath=&WBP_FM_Core_Button_SmallerRadius
   bIsVariable=true
   parent=SongOptionButtons
   widget=Controls
   properties {icon=&T-Icon-Controller-64, text=Controls}
   slot {class=VerticalBoxSlot, properties {padding=<ltrb 10 10 10 10>}}
  }
  Tips {
   widgetClassPath=&WBP_FM_Core_Button_SmallerRadius
   bIsVariable=true
   parent=SongOptionButtons
   widget=Tips
   properties {icon=/FMCore/Textures/T_Icon_Tutorial, text=Tips}
   slot {class=VerticalBoxSlot, properties {padding=<ltrb 10 10 10 10>}}
  }
  Settings {
   widgetClassPath=&WBP_FM_Core_Button_SmallerRadius
   bIsVariable=true
   parent=SongOptionButtons
   widget=Settings
   properties {icon=/Game/UI/Sidebar/Textures/T_UI_Icon_Sidebar_Settings, text=Settings}
   slot {class=VerticalBoxSlot, properties {padding=<ltrb 10 10 10 10>}}
  }
 ]
}
