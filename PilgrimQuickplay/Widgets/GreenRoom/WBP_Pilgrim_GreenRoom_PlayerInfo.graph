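# Green room row widget for Pilgrim Quickplay showing one player's avatar, name, platform or local-player label, instrument, difficulty and ready state.
# InitPlayerState clears old listeners, awaits playlist data and the event router, then calls SetPlayer to fill the row.
# Router events then update instrument, difficulty and readiness, with a ready-up animation played or reversed.
# The avatar texture loads asynchronously, retrying on a delay while its reference is invalid.
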WidgetBlueprint /PilgrimQuickplay/Widgets/GreenRoom/WBP_Pilgrim_GreenRoom_PlayerInfo
animations [Anim_PlayerReadyUp {
  binding_count=5
  length_seconds=1
  object=&self:Anim_PlayerReadyUp
  sequence {bindings [
    ReadyCheckmark {
     id=#1
     possessed_class=Overlay
     tracks [
      {
       type=MovieScene2DTransformTrack
       sections [{
         type=MovieScene2DTransformSection
         start_frame=0
         channels [
          Angle {type=MovieSceneScriptingFloatChannel, const [0,1]}
          Scale.X {type=MovieSceneScriptingFloatChannel, keys [[2,0],[4,1.125],[8,1]]}
          Scale.Y {type=MovieSceneScriptingFloatChannel, keys [[2,0],[4,1.125],[8,1]]}
          Shear.X {type=MovieSceneScriptingFloatChannel, const [0,1]}
          Shear.Y {type=MovieSceneScriptingFloatChannel, const [0,1]}
         ]
        }]
      }
      {
       type=MovieSceneFloatTrack
       display_name=Render Opacity
       sections [{type=MovieSceneFloatSection, channels [None {type=MovieSceneScriptingFloatChannel, keys [[2,0],[4,1]]}]}]
      }
     ]
    }
    Background {
     id=#2
     possessed_class=Image
     tracks [
      {
       type=MovieSceneWidgetMaterialTrack
       sections [{
         type=MovieSceneParameterSection
         channels [
          Alpha_Final {type=MovieSceneScriptingFloatChannel, keys [[0,0],[2,2],[20,1]]}
          UseEaseCurve {type=MovieSceneScriptingFloatChannel, keys [[0,1],[4,0],[20,0.5]]}
          VerticalFalloff {type=MovieSceneScriptingFloatChannel, keys [[0,0.2],[4,0.5],[10,0.2]]}
         ]
        }]
      }
      {
       type=MovieScene2DTransformTrack
       sections [{
         type=MovieScene2DTransformSection
         channels [
          Translation.X {type=MovieSceneScriptingFloatChannel, keys [[0,-256],[8,-128]]}
          Translation.Y {type=MovieSceneScriptingFloatChannel, const [0,1]}
          Scale.X {type=MovieSceneScriptingFloatChannel, const [2,1]}
          Scale.Y {type=MovieSceneScriptingFloatChannel, keys [[0,1],[2,1.25],[10,1]]}
          Shear.X {type=MovieSceneScriptingFloatChannel, const [0,1]}
          Shear.Y {type=MovieSceneScriptingFloatChannel, const [0,1]}
         ]
        }]
      }
     ]
    }
    PlayerAvatar {
     id=#3
     possessed_class=Overlay
     tracks [{
       type=MovieScene2DTransformTrack
       sections [{
         type=MovieScene2DTransformSection
         end_frame=7
         start_frame=0
         channels [
          Translation.X {type=MovieSceneScriptingFloatChannel, const [0,1]}
          Translation.Y {type=MovieSceneScriptingFloatChannel, const [0,1]}
          Angle {type=MovieSceneScriptingFloatChannel, const [0,1]}
          Scale.X {type=MovieSceneScriptingFloatChannel, const [1,1]}
          Scale.Y {type=MovieSceneScriptingFloatChannel, const [1,1]}
          Shear.X {type=MovieSceneScriptingFloatChannel, const [0,1]}
          Shear.Y {type=MovieSceneScriptingFloatChannel, const [0,1]}
         ]
        }]
      }]
    }
    Pulse {
     id=#4
     possessed_class=Image
     tracks [
      {
       type=MovieSceneByteTrack
       display_name=Visibility
       sections [{type=MovieSceneByteSection, end_frame=20, start_frame=0, channels [None {type=MovieSceneScriptingByteChannel, keys [[0,4],[20,2]]}]}]
      }
      {
       type=MovieSceneFloatTrack
       display_name=Render Opacity
       sections [{type=MovieSceneFloatSection, channels [None {type=MovieSceneScriptingFloatChannel, keys [[0,0],[2,0.25],[15,0]]}]}]
      }
      {
       type=MovieScene2DTransformTrack
       sections [{
         type=MovieScene2DTransformSection
         channels [
          Scale.X {type=MovieSceneScriptingFloatChannel, keys [[0,1.5],[20,3]]}
          Scale.Y {type=MovieSceneScriptingFloatChannel, keys [[0,1.5],[20,3]]}
         ]
        }]
      }
     ]
    }
    ReadyFX {
     id=#5
     possessed_class=Image
     tracks [
      {
       type=MovieSceneByteTrack
       display_name=Visibility
       sections [{type=MovieSceneByteSection, channels [None {type=MovieSceneScriptingByteChannel, keys [[0,4],[14,2]]}]}]
      }
      {
       type=MovieSceneWidgetMaterialTrack
       sections [{type=MovieSceneParameterSection, channels [Alpha_Final {type=MovieSceneScriptingFloatChannel, keys [[0,0],[2,32],[14,0]]}]}]
      }
      {
       type=MovieScene2DTransformTrack
       sections [{
         type=MovieScene2DTransformSection
         channels [
          Translation.X {type=MovieSceneScriptingFloatChannel, keys [[0,-256],[8,-128]]}
          Translation.Y {type=MovieSceneScriptingFloatChannel, const [0,1]}
          Scale.X {type=MovieSceneScriptingFloatChannel, const [2,1]}
          Scale.Y {type=MovieSceneScriptingFloatChannel, keys [[0,1],[2,1.25],[10,1]]}
          Shear.X {type=MovieSceneScriptingFloatChannel, const [0,1]}
          Shear.Y {type=MovieSceneScriptingFloatChannel, const [0,1]}
         ]
        }]
      }
     ]
    }
   ]}
 }]
graphs [
 CheckForMatchingPlayerState {nodes [
   n1 = CheckForMatchingPlayerState!() -> (then, FortPlayerPawn:Fort Player Pawn Object Reference) {is=title K2Node_FunctionEntry}
   > Return Node!(
    execute
    PlayerStateMatch:bool="Equal (Object)"(
     A=Get PlayerState!(self:Pawn Object Reference=@n1) -> (PlayerState:Player State Object Reference)
     B=Get MyPlayerState() -> (MyPlayerState:Player State Object Reference)
    )
   )
  ]}
 EventGraph {nodes [
   n1 = SetPlayerImage!() -> (OutputDelegate, then, FortPlayerStateAthena:Fort Player State Athena Object Reference) {is=title K2Node_CustomEvent}
   > n2 = GetDynamicMaterial(self=Get PlayerImage() -> (PlayerImage:Image Object Reference))
   > SetVectorParameterValue(
    self=@n2
    ParameterName=HighlightRGB
    Value=GetPinColor!(self:Fort Player State Zone Object Reference=@n1.FortPlayerStateAthena) -> (ReturnValue:Linear Color Structure)
   )
   > SetScalarParameterValue(self=@n2, ParameterName=BackgroundAlpha, Value=0.5)
   > n3 = Set PlayerIconRetryCount(PlayerIconRetryCount:int=Get PlayerIconRetryCount_Default() -> (PlayerIconRetryCount_Default:int=0)) -> (Output_Get:int=0)
   n4 = GetSocialAvatarBrush!(self:Fort Player State Zone Object Reference=@n1.FortPlayerStateAthena, bSmallImage:bool=true) -> (ReturnValue:Texture 2D Soft Object Reference)
   n5 = Branch(execute=[@n9.then, @n3.then], Condition=IsValidSoftObjectReference(SoftObjectReference=@n4))
   > n6 = Async Load Asset(Asset=@n4)
   n7 = Cast To Texture2D(execute=@n6.Completed, Object=@n6)
   > SetTextureParameterValue(self=@n2, ParameterName=SmallPreview, Value=@n7)
   n8 = Branch(execute=@n5.else, Condition="integer > integer"(A=Get PlayerIconRetryCount() -> (PlayerIconRetryCount:int=0)))
   > Decrement Int(Value=Get PlayerIconRetryCount() -> (PlayerIconRetryCount=0))
   > n9 = Delay(Duration=1)
   PrintString!(execute=@n8.else, InString=Failed to get a vaild texture reference for player, bPrintToScreen, bPrintToLog, TextColor, Duration, Key) -> (then)
   n10 = InitPlayerState!() -> (OutputDelegate, then, PlayerState:Player State Object Reference) {is=title K2Node_CustomEvent}
   > n11 = Branch(Condition=Is Not Empty!(TargetArray:Array of Async Action Start Listening To Event Object References=Get PlayerStateEventHandles() -> (PlayerStateEventHandles:Array of Async Action Start Listening To Event Object References)) -> (ReturnValue:bool))
   > n12 = For Each Loop(Array:Array of Async Action Start Listening To Event Object References=Get PlayerStateEventHandles()) -> (Array Element:Async Action Start Listening To Event Object Reference)
   > Is Valid(InputObject=@n12)
   > StopListeningToEvent(self:Async Action Start Listening To Event Object Reference=@n12)
   n13 = Clear(execute=@n12.Completed, TargetArray:Array of Async Action Start Listening To Event Object References=Get PlayerStateEventHandles())
   n14 = PlaylistDataReadyAsync!(execute:Exec=[@n11.else, @n13.then]) -> (
    then:Exec
    OnReady:Exec
    GameState:Fort Game State Athena Object Reference
    Playlist:Fort Playlist Object Reference
    PlaylistContextTags:"Gameplay Tag Container Structure (by ref)"
   ) {is=title K2Node_AsyncAction}
   n15 = PilgrimWaitForEventRouter!(execute:Exec=@n14.OnReady, ContextActor:Actor Object Reference=GetOwningPlayer()) -> (then:Exec, OnEventRouterAvailable:Exec, EventRouter:Gameplay Event Router Component Object Reference) {is=title K2Node_AsyncAction}
   Set Event Router(execute=@n15.OnEventRouterAvailable, Event Router:Gameplay Event Router Component Object Reference=@n15) -> (Output_Get:Gameplay Event Router Component Object Reference)
   > n16 = SetPlayer!(execute, self, PlayerState:Player State Object Reference=@n10.PlayerState) -> (then, PlayerAdded:bool)
   > Branch(Condition=@n16)
   > n17 = Sequence!(execute) -> (then_0, then_1, then_2:Exec)
   n18 = Get MyPlayerState()
   > n19 = StartListeningToEvent!(
    execute:Exec
    Target:Gameplay Event Router Component Object Reference=Get Event Router() -> (Event Router:Gameplay Event Router Component Object Reference)
    EventType:Script Struct Object Reference=/Script/PilgrimCoreRuntime.PilgrimEvent_PlayerState_InstrumentChanged
    Context:Object Reference=@n18
    EventBubblingRule:EEventBubblingRule Enum
   ) -> (
    then:Exec
    AsyncTaskProxy:Async Action Start Listening To Event Object Reference
    OnEventReceived:Exec
    Payload_PlayerState:Player State Object Reference
    Payload_NewInstrumentChoice:EPilgrimTrackType Enum
    Payload_PrevInstrumentChoice:EPilgrimTrackType Enum
   )
   > Add(
    TargetArray:Array of Async Action Start Listening To Event Object References=Get PlayerStateEventHandles()
    NewItem:"Async Action Start Listening To Event Object Reference (by ref)"=@n19
   ) {is=title K2Node_CallArrayFunction}
   Set Instrument Type!(
    execute=@n19.OnEventReceived
    self
    TrackType:EPilgrimTrackType Enum=@n19.Payload_NewInstrumentChoice
    PlayerState:Player State Object Reference=@n18
   ) -> (then) {is=title K2Node_CallFunction}
   n20 = StartListeningToEvent!(
    execute=@n17.then_2
    Target=Get Event Router()
    EventType=/Script/PilgrimCoreRuntime.PilgrimEvent_PlayerState_ReadinessChanged
    Context
    EventBubblingRule
   ) -> (then, AsyncTaskProxy, OnEventReceived, Payload_ForPlayer:Fort Player Pawn Object Reference, Payload_bIsReady:bool=~)
   > Add(TargetArray=Get PlayerStateEventHandles(), NewItem=@n20)
   Branch(execute=@n20.OnEventReceived, Condition="Equal (Object)"(A=Get PlayerState(self=@n20.Payload_ForPlayer), B=Get MyPlayerState()))
   > Set Ready State UI!(execute, self, IsReady:bool=@n20.Payload_bIsReady) -> (then) {is=title K2Node_CallFunction}
   n21 = StartListeningToEvent!(
    execute=@n17.then_1
    Target=Get Event Router()
    EventType=/Script/PilgrimCoreRuntime.PilgrimEvent_PlayerState_DifficultyChanged
    Context
    EventBubblingRule
   ) -> (
    then
    AsyncTaskProxy
    OnEventReceived
    Payload_PlayerState:Player State Object Reference
    Payload_NewSongDifficulty:EPilgrimSongDifficulty Enum
    Payload_PrevSongDifficulty:EPilgrimSongDifficulty Enum
   )
   > Add(TargetArray=Get PlayerStateEventHandles(), NewItem=@n21)
   Branch(execute=@n21.OnEventReceived, Condition="Equal (Object)"(A=@n21.Payload_PlayerState, B=Get MyPlayerState()))
   > SetDifficultyDisplay!(execute, self, Song Difficulty:EPilgrimSongDifficulty Enum=@n21.Payload_NewSongDifficulty) -> (then)
   Event Construct()
   > HidePlayerDisplay()
   n22 = Get Anim_PlayerReadyUp() -> (Anim_PlayerReadyUp:Widget Animation Object Reference)
   > PlayAnimation(InAnimation=@n22, StartAtTime=GetEndTime(self=@n22), PlayMode=Reverse)
   Event Destruct()
   > Parent: Destruct()
   n23 = Get PlayerStateEventHandles()
   > n24 = For Each Loop(Array=@n23)
   Clear(execute=@n24.Completed, TargetArray=@n23)
   Is Valid(exec=@n24.LoopBody, InputObject=@n24)
   > StopListeningToEvent(self=@n24)
  ]}
 HidePlayerDisplay {nodes [
   HidePlayerDisplay() {is=title K2Node_FunctionEntry}
   > SetVisibility(self=Self-Reference(), InVisibility=Hidden)
  ]}
 Set Instrument Type {nodes [
   n1 = Set Instrument Type!() -> (then, TrackType:EPilgrimTrackType Enum=TrackGuitar, PlayerState:Player State Object Reference) {is=title K2Node_FunctionEntry}
   n2 = GetCurrentSong!(ContextActor:Actor Object Reference=@n1.PlayerState) -> (ReturnValue:Pilgrim Song Metadata Object Reference)
   > n3 = Is Valid(InputObject=@n2)
   > UpdateInstrumentIcon!(
    execute
    self
    TrackType:EPilgrimTrackType Enum=@n1
    StartingInstrument:ESparksInstrumentType Enum=GetStartingInstrumentType!(self:Pilgrim Song Metadata Object Reference=@n2, TrackType:EPilgrimTrackType Enum=@n1) -> (ReturnValue:ESparksInstrumentType Enum=Guitar)
   ) -> (then)
   n4 = GetPilgrimQuickplayPlayerStateComponent!(execute=@n3.Is Not Valid, PlayerState:Player State Object Reference=@n1.PlayerState) -> (then, ReturnValue:Pilgrim Quickplay Player State Component Object Reference)
   > UpdateInstrumentIcon(
    TrackType=GetInstrumentChoice!(self:Pilgrim Core Player State Component Object Reference=@n4) -> (ReturnValue:EPilgrimTrackType Enum=TrackGuitar)
    StartingInstrument=GetSparksInstrumentChoice!(self:Sparks Player State Component Object Reference=@n4) -> (ReturnValue:ESparksInstrumentType Enum=Guitar)
   )
  ]}
 Set Ready State UI {nodes [
   n1 = Set Ready State UI!() -> (then, IsReady:bool=~) {is=title K2Node_FunctionEntry}
   > n2 = Branch(Condition=@n1)
   > SetVisibility(
    self=[Get ReadyCheckmark() -> (ReadyCheckmark:Overlay Object Reference), Get Background() -> (Background:Image Object Reference)]
    InVisibility=SelfHitTestInvisible
   )
   > PlayAnimation(InAnimation=Get Anim_PlayerReadyUp())
   SetVisibility(execute=@n2.else, self=[Get Background(), Get ReadyCheckmark()], InVisibility=Hidden)
   n3 = Get Anim_PlayerReadyUp()
   > PlayAnimation(InAnimation=@n3, StartAtTime=GetEndTime(self=@n3), PlayMode=Reverse)
  ]}
 SetDifficultyDisplay {nodes [
   n1 = SetDifficultyDisplay!() -> (then, Song Difficulty:EPilgrimSongDifficulty Enum=DifficultyEasy) {is=title K2Node_FunctionEntry}
   > SetDifficulty!(
    execute
    self:WBP Pilgrim Results Difficulty Object Reference=Get WBP_Pilgrim_Results_Difficulty() -> (WBP_Pilgrim_Results_Difficulty:WBP Pilgrim Results Difficulty Object Reference)
    Difficulty:EPilgrimSongDifficulty Enum=@n1
   ) -> (then)
  ]}
 SetPlayer {nodes [
   n1 = SetPlayer!() -> (then, PlayerState:Player State Object Reference) {is=title K2Node_FunctionEntry}
   > n2 = Sequence()
   > Is Valid(InputObject=@n1)
   > ShowPlayerDisplay()
   > n3 = Set MyPlayerState(MyPlayerState:Player State Object Reference=@n1) -> (Output_Get:Player State Object Reference)
   > "SetText (Text)"(
    self=Get PlayerName() -> (PlayerName:Common Text Object Reference)
    InText="To Text (String)"(InString=GetPlayerNameSafe!(Actor:Actor Object Reference=@n3) -> (ReturnValue:String))
   )
   > n4 = Cast To FortPlayerStateAthena!(execute, Object=@n3) -> (then, CastFailed, AsFort Player State Athena:Fort Player State Athena Object Reference) {is=title K2Node_DynamicCast}
   > SetPlatformPrefixIcon!(
    execute
    Image:Image Object Reference=Get Image_PlatformIcon() -> (Image_PlatformIcon:Image Object Reference)
    OtherPlayerPlatform:String=GetPlatform!(self:Fort Player State Object Reference=@n4) -> (ReturnValue:String)
   ) -> (then)
   > n5 = Sequence()
   > SetPlayerImage!(execute, self, FortPlayerStateAthena:Fort Player State Athena Object Reference=@n4) -> (then) {is=title K2Node_CallFunction}
   n6 = GetPilgrimQuickplayPlayerStateComponent(execute=@n5.then_1, PlayerState=Get MyPlayerState())
   > n7 = Is Valid(InputObject=@n6)
   FortShippingLog!(
    execute=@n7.Is Not Valid
    WorldContextObject:Object Reference
    InString:String=Quickplay Player State Component not found for PlayerState assigned to WBP_Pilgrim_GreenRoom_PlayerInfo
    bLogAsWarning:bool=false
   ) -> (then)
   SetDifficultyDisplay(
    execute=@n7.Is Valid
    Song Difficulty=GetSongDifficulty!(self:Pilgrim Core Player State Component Object Reference=@n6) -> (ReturnValue:EPilgrimSongDifficulty Enum=DifficultyEasy)
   ) {is=title K2Node_CallFunction}
   > Set Instrument Type(TrackType=GetInstrumentChoice(self=@n6) -> (ReturnValue=TrackGuitar), PlayerState=Get MyPlayerState()) {is=title K2Node_CallFunction}
   > Set Ready State UI(IsReady=GetIsReady!(self:Pilgrim Core Player State Component Object Reference=@n6) -> (ReturnValue:bool)) {is=title K2Node_CallFunction}
   > Set WasPlayerAdded(WasPlayerAdded:bool=true) -> (Output_Get:bool)
   SetLocalPlayerIndexFromPlayerState!(
    execute=@n5.then_2
    self:WBP Pilgrim Player Identity Local Identity Label Object Reference=Get PlayerIdentity_LocalIdentity() -> (PlayerIdentity_LocalIdentity:WBP Pilgrim Player Identity Local Identity Label Object Reference)
    Player State:Player State Object Reference=Get MyPlayerState()
   ) -> (then)
   > ShowLocalIdentityOrPlatform!(execute, self, IsLocalMPView:bool=IsHamSandwichSessionActive!() -> (ReturnValue:bool)) -> (then)
   FortShippingLog(execute=@n4.CastFailed, InString=PlayerState assigned to WBP_Pilgrim_GreenRoom_PlayerInfo did not cast to a valid type, bLogAsWarning=false)
   Return Node!(execute=@n2.then_2, PlayerAdded:bool=Get WasPlayerAdded() -> (WasPlayerAdded:bool))
   n8 = Branch(execute=@n2.then_1, Condition=Get WasPlayerAdded())
   HidePlayerDisplay(execute=@n8.else) {is=title K2Node_CallFunction}
  ]}
 SetPlayerVote {nodes [SetPlayerVote!() -> (then, PilgrimSongMetadata:Pilgrim Song Metadata Object Reference) {is=title K2Node_FunctionEntry}]}
 SetToDefault {nodes [SetToDefault() {is=title K2Node_FunctionEntry}]}
 ShowLocalIdentityOrPlatform {nodes [
   ShowLocalIdentityOrPlatform!() -> (then, IsLocalMPView:bool=~) {is=title K2Node_FunctionEntry}
   n1 = Get MyPlayerState()
   > n2 = Is Valid(InputObject=@n1)
   > n3 = Branch(Condition=Is Local Player in Local MP Session!(PlayerState:Player State Object Reference=@n1) -> (ReturnValue:bool))
   n4 = Get VS_LocalOrOnlineID() -> (VS_LocalOrOnlineID:Common Visibility Switcher Object Reference)
   > SetActiveWidgetIndex(self:Common Visibility Switcher Object Reference=@n4, Index=1)
   SetActiveWidgetIndex(execute=[@n3.else, @n2.Is Not Valid], self=@n4, Index=0)
  ]}
 ShowPlayerDisplay {nodes [
   ShowPlayerDisplay() {is=title K2Node_FunctionEntry}
   > SetVisibility(self=Self-Reference(), InVisibility=SelfHitTestInvisible)
  ]}
 UpdateInstrumentIcon {nodes [
   n1 = UpdateInstrumentIcon!() -> (then, TrackType:EPilgrimTrackType Enum=TrackGuitar, StartingInstrument:ESparksInstrumentType Enum=Guitar) {is=title K2Node_FunctionEntry}
   > n2 = GetDynamicMaterial(self=Get InstrumentImage() -> (InstrumentImage:Image Object Reference))
   > SetScalarParameterValue(
    self=@n2
    ParameterName=Icon
    Value="To Float (Integer)"(InInt=GetInstrumentIconIndex!(
      InstrumentType:EPilgrimTrackType Enum=@n1
      SongMetaData:Pilgrim Song Metadata Object Reference
      StartingInstrumentIndex:ESparksInstrumentType Enum=@n1.StartingInstrument
      bUseKeyTarIcon:bool=false
     ) -> (InstrumentIdexIcon:int=0))
   )
  ]}
]
variables [
 Event Router {type=GameplayEventRouterComponent, default=null}
 MyPlayerState {type=PlayerState, default=null}
 PlayerIconRetryCount {type=integer, default=10}
 PlayerIconRetryCount_Default {type=integer, default=10}
 PlayerStateEventHandles {type="array<AsyncAction_StartListeningToEvent>", default []}
 PlayspacePlayerStateIndex {type=integer}
]
widgets {
 info {inheritedWidgetCount=0, parentClass=&UserWidget, rootWidgetClass=&Overlay, widgetCount=22}
 widgets [
  Overlay_11 {widgetClassPath=&Overlay, widget=Overlay_11}
  Background {
   widgetClassPath=&Image
   bIsVariable=true
   parent=Overlay_11
   widget=Background
   properties {
    clipping=ClipToBoundsWithoutIntersecting
    visibility=Hidden
    brush {
     bIsDynamicallyLoaded=false
     drawAs=Image
     imageType=NoImage
     mirroring=NoMirror
     resourceName=None
     resourceObject=/PilgrimCore/Widgets/Materials/MI_UI_Pilgrim_BandReady_FlareGradient
     tiling=NoTile
     imageSize=<64 48>
     margin=<ltrb 0 0 0 0>
     outlineSettings {
      bUseBrushTransparency=false
      roundingType=HalfHeightRadius
      width=0
      color {colorUseRule=UseColor_Specified, specifiedColor=<rgba 0 0 0 0>}
      cornerRadii=<0 0 0 0>
     }
     tintColor {colorUseRule=UseColor_Specified, specifiedColor=<rgba 1 1 1 0.65>}
     uVRegion {bIsValid=false, max=<0 0>, min=<0 0>}
    }
    renderTransform {angle=0, scale=<1.5 1>, shear=<0 0>, translation=<-80 0>}
    renderTransformPivot=<0 0.5>
   }
   slot {class=OverlaySlot, properties {horizontalAlignment=HAlign_Fill, verticalAlignment=VAlign_Center}}
  }
  ReadyFX {
   widgetClassPath=&Image
   bIsVariable=true
   parent=Overlay_11
   widget=ReadyFX
   properties {
    clipping=ClipToBoundsWithoutIntersecting
    visibility=Hidden
    brush {
     bIsDynamicallyLoaded=false
     drawAs=Image
     imageType=NoImage
     mirroring=NoMirror
     resourceName=None
     resourceObject=/PilgrimCore/Widgets/Materials/MI_UI_Pilgrim_BandReady_Bandmate_Flare
     tiling=NoTile
     imageSize=<64 32>
     margin=<ltrb 0 0 0 0>
     outlineSettings {
      bUseBrushTransparency=false
      roundingType=HalfHeightRadius
      width=0
      color {colorUseRule=UseColor_Specified, specifiedColor=<rgba 0 0 0 0>}
      cornerRadii=<0 0 0 0>
     }
     tintColor {colorUseRule=UseColor_Specified, specifiedColor=<rgba 1 1 1 0.65>}
     uVRegion {bIsValid=false, max=<0 0>, min=<0 0>}
    }
    renderTransform {angle=0, scale=<3 1>, shear=<0 0>, translation=<-128 0>}
    renderTransformPivot=<0 0.5>
   }
   slot {class=OverlaySlot, properties {horizontalAlignment=HAlign_Fill, verticalAlignment=VAlign_Center}}
  }
  HorizontalBox_0 {
   widgetClassPath=&HorizontalBox
   parent=Overlay_11
   widget=HorizontalBox_0
   properties {renderTransformPivot=<0 0.5>}
   slot {class=OverlaySlot, properties {horizontalAlignment=HAlign_Fill, verticalAlignment=VAlign_Center}}
  }
  PlayerAvatar {
   widgetClassPath=&Overlay
   parent=HorizontalBox_0
   widget=PlayerAvatar
   slot {class=HorizontalBoxSlot, properties {horizontalAlignment=HAlign_Left, verticalAlignment=VAlign_Center, padding=<ltrb 24 0 8 0>}}
  }
  PlayerShadow {
   widgetClassPath=&Image
   bIsVariable=true
   parent=PlayerAvatar
   widget=PlayerShadow
   properties {
    visibility=SelfHitTestInvisible
    brush {
     bIsDynamicallyLoaded=false
     drawAs=Box
     imageType=NoImage
     mirroring=NoMirror
     resourceName=None
     resourceObject=/PilgrimCore/Widgets/Materials/M_UI_Pilgrim_RoundedShadow_Round
     tiling=NoTile
     imageSize=<64 64>
     margin=<ltrb 0.5 0.5 0.5 0.5>
     outlineSettings {
      bUseBrushTransparency=false
      roundingType=HalfHeightRadius
      width=0
      color {colorUseRule=UseColor_Specified, specifiedColor=<rgba 0 0 0 0>}
      cornerRadii=<0 0 0 0>
     }
     tintColor {colorUseRule=UseColor_Specified, specifiedColor=<rgba 1 1 1 1>}
     uVRegion {bIsValid=false, max=<0 0>, min=<0 0>}
    }
    renderTransform {angle=0, scale=<1.5 1.5>, shear=<0 0>, translation=<0 0>}
   }
   slot {class=OverlaySlot, properties {horizontalAlignment=HAlign_Fill, verticalAlignment=VAlign_Fill}}
  }
  Pulse {
   widgetClassPath=&Image
   bIsVariable=true
   parent=PlayerAvatar
   widget=Pulse
   properties {
    clipping=ClipToBoundsWithoutIntersecting
    visibility=Hidden
    brush {
     bIsDynamicallyLoaded=false
     drawAs=Box
     imageType=NoImage
     mirroring=NoMirror
     resourceName=None
     resourceObject=/PilgrimCore/Widgets/Materials/MI_UI_Piligrim_PlayerReady_Pulse
     tiling=NoTile
     imageSize=<64 64>
     margin=<ltrb 0.5 0.5 0.5 0.5>
     outlineSettings {
      bUseBrushTransparency=false
      roundingType=HalfHeightRadius
      width=0
      color {colorUseRule=UseColor_Specified, specifiedColor=<rgba 0 0 0 0>}
      cornerRadii=<0 0 0 0>
     }
     tintColor {colorUseRule=UseColor_Specified, specifiedColor=<rgba 1 1 1 1>}
     uVRegion {bIsValid=false, max=<0 0>, min=<0 0>}
    }
    renderTransform {angle=0, scale=<1.5 1.5>, shear=<0 0>, translation=<0 0>}
   }
   slot {class=OverlaySlot, properties {horizontalAlignment=HAlign_Fill, verticalAlignment=VAlign_Fill}}
  }
  PlayerImage {
   widgetClassPath=&Image
   bIsVariable=true
   parent=PlayerAvatar
   widget=PlayerImage
   properties {brush {
     bIsDynamicallyLoaded=false
     drawAs=Image
     imageType=NoImage
     mirroring=NoMirror
     resourceName=None
     resourceObject=/Game/UI/SocialPanel/Art/Materials/UI_UserPortrait_Face_NoBackground
     tiling=NoTile
     imageSize=<64 64>
     margin=<ltrb 0 0 0 0>
     outlineSettings {
      bUseBrushTransparency=false
      roundingType=HalfHeightRadius
      width=0
      color {colorUseRule=UseColor_Specified, specifiedColor=<rgba 0 0 0 0>}
      cornerRadii=<0 0 0 0>
     }
     tintColor {colorUseRule=UseColor_Specified, specifiedColor=<rgba 1 1 1 1>}
     uVRegion {bIsValid=false, max=<0 0>, min=<0 0>}
    }}
   slot {class=OverlaySlot, properties {horizontalAlignment=HAlign_Fill, verticalAlignment=VAlign_Fill}}
  }
  ReadyCheckmark {
   widgetClassPath=&Overlay
   bIsVariable=true
   parent=PlayerAvatar
   widget=ReadyCheckmark
   properties {visibility=Hidden, renderTransform {angle=0, scale=<1 1>, shear=<0 0>, translation=<0 4>}}
   slot {class=OverlaySlot, properties {horizontalAlignment=HAlign_Center, verticalAlignment=VAlign_Bottom}}
  }
  Check_Glow {
   widgetClassPath=&Image
   parent=ReadyCheckmark
   widget=Check_Glow
   properties {
    visibility=SelfHitTestInvisible
    brush {
     bIsDynamicallyLoaded=false
     drawAs=Image
     imageType=NoImage
     mirroring=NoMirror
     resourceName=None
     resourceObject=/PilgrimCore/Widgets/Materials/MI_UI_PilgrimReadyGlow
     tiling=NoTile
     imageSize=<24 24>
     margin=<ltrb 0 0 0 0>
     outlineSettings {
      bUseBrushTransparency=false
      roundingType=HalfHeightRadius
      width=0
      color {colorUseRule=UseColor_Specified, specifiedColor=<rgba 0 0 0 0>}
      cornerRadii=<0 0 0 0>
     }
     tintColor {colorUseRule=UseColor_Specified, specifiedColor=<rgba 1 1 1 1>}
     uVRegion {bIsValid=false, max=<0 0>, min=<0 0>}
    }
    renderTransform {angle=0, scale=<2.5 2.5>, shear=<0 0>, translation=<0 0>}
   }
   slot {class=OverlaySlot, properties {horizontalAlignment=HAlign_Fill, verticalAlignment=VAlign_Fill}}
  }
  Check_Pulse {
   widgetClassPath=&Image
   parent=ReadyCheckmark
   widget=Check_Pulse
   properties {
    visibility=SelfHitTestInvisible
    brush {
     bIsDynamicallyLoaded=false
     drawAs=Image
     imageType=NoImage
     mirroring=NoMirror
     resourceName=None
     resourceObject=/PilgrimCore/Widgets/Materials/MI_UI_PilgrimReadyStatePulse
     tiling=NoTile
     imageSize=<24 24>
     margin=<ltrb 0 0 0 0>
     outlineSettings {
      bUseBrushTransparency=false
      roundingType=HalfHeightRadius
      width=0
      color {colorUseRule=UseColor_Specified, specifiedColor=<rgba 0 0 0 0>}
      cornerRadii=<0 0 0 0>
     }
     tintColor {colorUseRule=UseColor_Specified, specifiedColor=<rgba 1 1 1 1>}
     uVRegion {bIsValid=false, max=<0 0>, min=<0 0>}
    }
    renderTransform {angle=0, scale=<2.5 2.5>, shear=<0 0>, translation=<0 0>}
   }
   slot {class=OverlaySlot, properties {horizontalAlignment=HAlign_Fill, verticalAlignment=VAlign_Fill}}
  }
  StatusIndicator {
   widgetClassPath=/FortUILibrary/UIKit/StatusIndicator/WBP_UIKit_StatusIndicator
   bIsVariable=true
   parent=ReadyCheckmark
   widget=StatusIndicator
   properties {icon=/FortUILibrary/Textures/MaterialInstances/MI_UIKit_Icon_Check, iconSize=Compact, overrideStatusIndicatorVisibility=true}
   slot {class=OverlaySlot}
  }
  VS_LocalOrOnlineID {
   widgetClassPath=&CommonVisibilitySwitcher
   bIsVariable=true
   parent=HorizontalBox_0
   widget=VS_LocalOrOnlineID
   slot {class=HorizontalBoxSlot, properties {horizontalAlignment=HAlign_Center, verticalAlignment=VAlign_Center, padding=<ltrb 0 0 5 0>}}
  }
  SB_PlatformIcon {
   widgetClassPath=&SizeBox
   parent=VS_LocalOrOnlineID
   widget=SB_PlatformIcon
   properties {bOverride_HeightOverride=true, bOverride_WidthOverride=true, heightOverride=22, widthOverride=22}
   slot {class=CommonVisibilitySwitcherSlot}
  }
  Image_PlatformIcon {
   widgetClassPath=&Image
   bIsVariable=true
   parent=SB_PlatformIcon
   widget=Image_PlatformIcon
   properties {
    visibility=Collapsed
    brush {
     bIsDynamicallyLoaded=false
     drawAs=Image
     imageType=NoImage
     mirroring=NoMirror
     resourceName=None
     resourceObject=/Game/UI/Friends_UI/Social/Console_PlatformIcon_32x
     tiling=NoTile
     imageSize=<32 32>
     margin=<ltrb 0 0 0 0>
     outlineSettings {
      bUseBrushTransparency=false
      roundingType=HalfHeightRadius
      width=0
      color {colorUseRule=UseColor_Specified, specifiedColor=<rgba 0 0 0 0>}
      cornerRadii=<0 0 0 0>
     }
     tintColor {colorUseRule=UseColor_Specified, specifiedColor=<rgba 1 1 1 1>}
     uVRegion {bIsValid=false, max=<0 0>, min=<0 0>}
    }
   }
   slot {class=SizeBoxSlot}
  }
  PlayerIdentity_LocalIdentity {
   widgetClassPath=/PilgrimCore/Widgets/Player/WBP_Pilgrim_PlayerIdentity_LocalIdentity_Label
   bIsVariable=true
   parent=VS_LocalOrOnlineID
   widget=PlayerIdentity_LocalIdentity
   properties {fontSizeOverride=22}
   slot {class=CommonVisibilitySwitcherSlot}
  }
  PlayerName {
   widgetClassPath=&CommonTextBlock
   bIsVariable=true
   parent=HorizontalBox_0
   widget=PlayerName
   properties {
    text="{PlayerName}"
    font {
     bForceMonospaced=false
     bMaterialIsStencil=false
     fontMaterial=null
     fontObject=&HeadingNow
     letterSpacing=0
     monospacedWidth=1
     size=16.5
     skewAmount=0
     typefaceFontName=75 Medium
     outlineSettings {
      bApplyOutlineToDropShadows=false
      bMiteredCorners=false
      bSeparateFillAlpha=false
      outlineMaterial=&M-FontShadow
      outlineSize=2
      outlineColor=<rgba 0 0 0 1>
     }
    }
    shadowOffset=<0 0>
   }
   slot {class=HorizontalBoxSlot, properties {horizontalAlignment=HAlign_Left, verticalAlignment=VAlign_Center, padding=<ltrb 0 0 4 0>}}
  }
  InstrumentOverlay {
   widgetClassPath=&Overlay
   parent=HorizontalBox_0
   widget=InstrumentOverlay
   slot {class=HorizontalBoxSlot, properties {horizontalAlignment=HAlign_Center, verticalAlignment=VAlign_Center, padding=<ltrb -4 0 -2 0>}}
  }
  InstrumentShadow {
   widgetClassPath=&Image
   bIsVariable=true
   parent=InstrumentOverlay
   widget=InstrumentShadow
   properties {
    visibility=SelfHitTestInvisible
    brush {
     bIsDynamicallyLoaded=false
     drawAs=Image
     imageType=NoImage
     mirroring=NoMirror
     resourceName=None
     resourceObject=&M_UI_Pilgrim_RoundedShadow_Round
     tiling=NoTile
     imageSize=<40 40>
     margin=<ltrb 0 0 0 0>
     outlineSettings {
      bUseBrushTransparency=false
      roundingType=HalfHeightRadius
      width=0
      color {colorUseRule=UseColor_Specified, specifiedColor=<rgba 0 0 0 0>}
      cornerRadii=<0 0 0 0>
     }
     tintColor {colorUseRule=UseColor_Specified, specifiedColor=<rgba 1 1 1 1>}
     uVRegion {bIsValid=false, max=<0 0>, min=<0 0>}
    }
   }
   slot {class=OverlaySlot, properties {horizontalAlignment=HAlign_Fill, verticalAlignment=VAlign_Fill}}
  }
  InstrumentImage {
   widgetClassPath=&Image
   bIsVariable=true
   parent=InstrumentOverlay
   widget=InstrumentImage
   properties {brush {
     bIsDynamicallyLoaded=false
     drawAs=Image
     imageType=FullColor
     mirroring=NoMirror
     resourceName=None
     resourceObject=/PilgrimQuickplay/Widgets/Materials/M_UI_Icon_Instrument_Type_MSDF
     tiling=NoTile
     imageSize=<36 36>
     margin=<ltrb 0 0 0 0>
     outlineSettings {
      bUseBrushTransparency=false
      roundingType=HalfHeightRadius
      width=0
      color {colorUseRule=UseColor_Specified, specifiedColor=<rgba 0 0 0 0>}
      cornerRadii=<0 0 0 0>
     }
     tintColor {colorUseRule=UseColor_Specified, specifiedColor=<rgba 1 1 1 1>}
     uVRegion {bIsValid=false, max=<0 0>, min=<0 0>}
    }}
   slot {class=OverlaySlot, properties {horizontalAlignment=HAlign_Center, verticalAlignment=VAlign_Center, padding=<ltrb 7 7 7 7>}}
  }
  ScaleBox_0 {
   widgetClassPath=&ScaleBox
   parent=HorizontalBox_0
   widget=ScaleBox_0
   properties {stretch=UserSpecified, userSpecifiedScale=0.7}
   slot {class=HorizontalBoxSlot, properties {horizontalAlignment=HAlign_Center, verticalAlignment=VAlign_Center}}
  }
  WBP_Pilgrim_Results_Difficulty {
   widgetClassPath=/PilgrimQuickplay/Widgets/Results/WBP_Pilgrim_Results_Difficulty
   bIsVariable=true
   parent=ScaleBox_0
   widget=WBP_Pilgrim_Results_Difficulty
   slot {class=ScaleBoxSlot}
  }
 ]
}
