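# List-entry variant of the Pilgrim GreenRoom player info widget, letting a list view populate it with one player per row.
# When its list item is set, it casts the object to PilgrimPlayerInfoVM, breaks out the PlayerInfo struct and passes its PlayerState to InitPlayerState.
# It subclasses WBP_Pilgrim_GreenRoom_PlayerInfo and adds no widgets of its own, serving only as a list adapter.
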WidgetBlueprint /PilgrimQuickplay/Widgets/GreenRoom/WBP_Pilgrim_GreenRoom_PlayerInfo_ListEntry
graphs [EventGraph {nodes [
   n1 = Event OnListItemObjectSet()
   > n2 = Cast To PilgrimPlayerInfoVM!(execute, Object=@n1.ListItemObject) -> (then, CastFailed, AsPilgrim Player Info VM:Pilgrim Player Info VM Object Reference) {is=title K2Node_DynamicCast}
   > InitPlayerState!(
    execute
    self:WBP Pilgrim Green Room Player Info Object Reference
    PlayerState:Player State Object Reference=Break Pilgrim Player Info!(PilgrimPlayerInfo:"Pilgrim Player Info Structure (by ref)"=Get PlayerInfo!(self:Pilgrim Player Info VM Object Reference=@n2) -> (PlayerInfo:Pilgrim Player Info Structure)) -> (PlayerState:Fort Player State Weak object, PlayerController:Fort Player Controller Weak object, bIsLocalPlayer:bool=~) {is=title K2Node_BreakStruct}
   ) -> (then)
  ]}]
variables [
 As BP Player Bandwidth Entry Data {type=Object, default=null}
 ListItemObject {type=Object, default=null}
]
widgets {
 info {
  inheritedWidgetCount=0
  parentClass=&/WBP_Pilgrim_GreenRoom_PlayerInfo.WBP_Pilgrim_GreenRoom_PlayerInfo_C
  rootWidgetClass=null
  widgetCount=0
 }
 widgets []
}
