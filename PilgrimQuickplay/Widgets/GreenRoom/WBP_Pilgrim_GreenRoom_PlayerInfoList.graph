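# Green Room player info list widget for Pilgrim Quickplay, built on CommonUserWidget with a CommonListView root.
# The list uses the Green Room PlayerInfo list entry widget for its rows.
# GetPlayerInfoList takes all-player and remote-player view model arrays, returning remote players when the PilgrimHamSandwich view model reports IsHamSandwichView, otherwise all players.
# The event graph is empty, and the file does not show where that result feeds the list.
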WidgetBlueprint /PilgrimQuickplay/Widgets/GreenRoom/WBP_Pilgrim_GreenRoom_PlayerInfoList
graphs [
 EventGraph {links [], nodes []}
 GetPlayerInfoList {nodes [
   GetPlayerInfoList!() -> (then, AllPlayers:Array of Pilgrim Player Info VM Object References, RemotePlayers:Array of Pilgrim Player Info VM Object References) {is=title K2Node_FunctionEntry}
   > n1 = Branch(Condition=Get IsHamSandwichView!(self:Pilgrim Ham Sandwich VM Object Reference=Get PilgrimHamSandwichVM() -> (PilgrimHamSandwichVM:Pilgrim Ham Sandwich VM Object Reference)) -> (IsHamSandwichView:bool))
   > Return Node!(
    execute
    PlayerInfoList:Array of Pilgrim Player Info VM Object References=Get RemotePlayers() -> (RemotePlayers:Array of Pilgrim Player Info VM Object References)
   )
   Return Node(execute=@n1.else, PlayerInfoList=Get AllPlayers() -> (AllPlayers:Array of Pilgrim Player Info VM Object References))
  ]}
]
variables []
widgets {
 info {inheritedWidgetCount=0, parentClass=&CommonUserWidget, rootWidgetClass=&CommonListView, widgetCount=1}
 widgets [CommonListView {
   widgetClassPath=&CommonListView
   bIsVariable=true
   widget=CommonListView
   properties {
    clipping=Inherit
    entryWidgetClass=&/WBP_Pilgrim_GreenRoom_PlayerInfo_ListEntry.WBP_Pilgrim_GreenRoom_PlayerInfo_ListEntry_C
    numDesignerPreviewEntries=4
    verticalEntrySpacing=8
    visibility=SelfHitTestInvisible
   }
  }]
}
