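# Pilgrim Quickplay widget showing a horizontal list of bandmate ready icons, visible during Pregame, SongResults and SetResults and collapsed on other state events.
# On Construct it seeds the list from the owning player's band and binds playspace user added/removed events.
# Adds create entry data from player state and current ready flag, capped at MaxListItems; removals match cached player state among displayed entries.
# An event router listener, started on initialization, drives visibility and stops on Destruct.
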WidgetBlueprint /PilgrimQuickplay/Widgets/Results/WBP_Pilgrim_Quickplay_BandmateReadyDisplay
graphs [EventGraph {nodes [
   n1 = OnUserAdded!() -> (OutputDelegate, then, AddedUser:"Playspace User Structure (by ref)") {is=title K2Node_CustomEvent}
   > AddPlayerReadyIcon!(
    execute
    self
    PlayerState:Player State Object Reference=Break Playspace User!(PlayspaceUser:"Playspace User Structure (by ref)"=@n1.AddedUser) -> (
     UserId:Unique Net Id Structure
     LeafPlayspace:Playspace Weak object
     PlayerStateCached:Player State Weak object
     ControllerCached:Controller Weak object
    ) {is=title K2Node_BreakStruct}.PlayerStateCached
   ) -> (then)
   n2 = AddPlayerReadyIcon!() -> (OutputDelegate, then, PlayerState:Player State Object Reference) {is=title K2Node_CustomEvent}
   n3 = Get PlayerReadyIconList() -> (PlayerReadyIconList:Common List View Object Reference)
   > Branch(Condition="integer < integer"(A=GetNumItems!(self:List View Object Reference=@n3) -> (ReturnValue:int=0), B=Get MaxListItems() -> (MaxListItems:int=0)))
   > n4 = Cast To FortPlayerStateAthena!(execute, Object=@n2.PlayerState) -> (then, CastFailed, AsFort Player State Athena:Fort Player State Athena Object Reference) {is=title K2Node_DynamicCast}
   > n5 = GetPilgrimQuickplayPlayerStateComponent!(execute, PlayerState:Player State Object Reference=@n4) -> (then, ReturnValue:Pilgrim Quickplay Player State Component Object Reference)
   > n6 = Construct Object from Class(Class=&/BP_AvatarIcon_Ready_EntryData.BP_AvatarIcon_Ready_EntryData_C) -> (ReturnValue:BP Avatar Icon Ready Entry Data Object Reference)
   > Set!(
    execute
    EntryData_PlayerState_4_#1:Fort Player State Athena Object Reference=@n4
    EntryData_IsReady_2_#2:bool=GetIsReady!(self:Pilgrim Core Player State Component Object Reference=@n5) -> (ReturnValue:bool)
    self:BP Avatar Icon Ready Entry Data Object Reference=@n6
   ) -> (then, Output_Get:S Avatar Icon Ready Data Structure)
   > AddItem!(execute, self:List View Object Reference=Get PlayerReadyIconList(), Item:Object Reference=@n6) -> (then)
   n7 = RemovePlayerReadyIcon!() -> (OutputDelegate, then, PlayerState:Player State Object Reference) {is=title K2Node_CustomEvent}
   n8 = Get PlayerReadyIconList()
   > n9 = For Each Loop with Break!(
    Exec:Exec
    Array:Array of Object References=GetDisplayedEntryWidgets!(self:List View Base Object Reference=@n8) -> (ReturnValue:Array of User Widget Object References)
    Break:Exec=@n11.then
   ) -> (LoopBody:Exec, Array Element:Object Reference, Array Index:int, Completed:Exec)
   > n10 = Cast To WBP_PilgrimQuickplay_AvatarIcon_Ready_ListEntry!(execute, Object=@n9) -> (then, CastFailed, AsWBP Pilgrim Quickplay Avatar Icon Ready List Entry:WBP Pilgrim Quickplay Avatar Icon Ready List Entry Object Reference) {is=title K2Node_DynamicCast}
   > Branch(Condition="Equal (Object)"(
     A=@n7.PlayerState
     B=Get CachedPlayerState!(self:WBP Pilgrim Quickplay Avatar Icon Ready Object Reference=@n10) -> (CachedPlayerState:Fort Player State Athena Object Reference)
    ))
   > n11 = RemoveItem!(
    execute
    self:List View Object Reference=@n8
    Item:Object Reference=GetListItemObject!(UserObjectListEntry:User Object List Entry Interface=@n10) -> (ReturnValue:Object Reference)
   ) -> (then)
   n12 = OnUserRemoved!() -> (OutputDelegate, then, RemovedUser:"Playspace User Structure (by ref)") {is=title K2Node_CustomEvent}
   > RemovePlayerReadyIcon!(execute, self, PlayerState:Player State Object Reference=Break Playspace User(PlayspaceUser=@n12.RemovedUser).PlayerStateCached) -> (then) {is=title K2Node_CallFunction}
   Event OnInitialized()
   > n13 = PlaylistDataReadyAsync!(execute:Exec) -> (
    then:Exec
    OnReady:Exec
    GameState:Fort Game State Athena Object Reference
    Playlist:Fort Playlist Object Reference
    PlaylistContextTags:"Gameplay Tag Container Structure (by ref)"
   ) {is=title K2Node_AsyncAction}
   n14 = PilgrimWaitForEventRouter!(execute:Exec=@n13.OnReady, ContextActor:Actor Object Reference=GetOwningPlayer()) -> (then:Exec, OnEventRouterAvailable:Exec, EventRouter:Gameplay Event Router Component Object Reference) {is=title K2Node_AsyncAction}
   Set Event Router(execute=@n14.OnEventRouterAvailable, Event Router:Gameplay Event Router Component Object Reference=@n14) -> (Output_Get:Gameplay Event Router Component Object Reference)
   n15 = Get Event Router() -> (Event Router:Gameplay Event Router Component Object Reference)
   > n16 = StartListeningToEvent!(
    execute:Exec
    Target:Gameplay Event Router Component Object Reference=@n15
    EventType:Script Struct Object Reference=/Script/PilgrimQPSharedRuntime.PilgrimEvent_QuickplayState_Began
    Context:Object Reference
    EventBubblingRule:EEventBubblingRule Enum
    None:%=@n15
   ) -> (
    then:Exec
    AsyncTaskProxy:Async Action Start Listening To Event Object Reference
    OnEventReceived:Exec
    Payload_StateId:EPilgrimQuickplayState Enum
   )
   > Set QuickplayStateBeganEventHandle(QuickplayStateBeganEventHandle:Async Action Start Listening To Event Object Reference=@n16) -> (Output_Get:Async Action Start Listening To Event Object Reference)
   n17 = Branch(
    execute=@n16.OnEventReceived
    Condition=OR Boolean!(
     A="Equal (Enum)"(A:EPilgrimQuickplayState Enum=@n16.Payload_StateId, B:EPilgrimQuickplayState Enum=Pregame)
     B="Equal (Enum)"(A=@n16.Payload_StateId, B=SongResults)
     C:bool="Equal (Enum)"(A=@n16.Payload_StateId, B=SetResults)
    ) -> (ReturnValue)
   )
   > SetVisibility(InVisibility=SelfHitTestInvisible)
   SetVisibility(execute=@n17.else, InVisibility=Collapsed)
   Event Construct()
   > n18 = PlaylistDataReadyAsync()
   n19 = GetOwningPlayer()
   n20 = Is Valid(exec=@n18.OnReady, InputObject=@n19)
   PrintString!(
    execute=@n20.Is Not Valid
    InString=No valid owning player for BandmateReadyDisplay widget. There must be a valid owning player for this widget to operate correctly.
    bPrintToScreen
    bPrintToLog
    TextColor=(R=1,G=0.4611,B=0.017,A=1)
    Duration=5
    Key
   ) -> (then)
   n21 = GetSparksMusicPlayspaceOfClass!(
    PlayspaceType:Sparks Music Playspace Class Reference=/Script/PilgrimQuickplayRuntime.PilgrimQuickplayPlayspace
    Actor:Actor Object Reference=@n19
    bMatchChildTypes:bool=false
   ) -> (ReturnValue:Pilgrim Quickplay Playspace Object Reference)
   n22 = Is Valid(exec=@n20.Is Valid, InputObject=@n21)
   PrintString(
    execute=@n22.Is Not Valid
    InString=No valid PilgrimQuickplayPlayspace returned from GetSparksMusicPlayspaceOfClass
    TextColor=(R=1,G=0.807,B=0.1156,A=1)
    Duration=5
   )
   n23 = Sequence(execute=@n22.Is Valid)
   > n24 = Sequence()
   > Bind Event to Notify Playspace User Added(self:Playspace Object Reference=@n21, Delegate=@n1) {is=title K2Node_AddDelegate}
   Bind Event to Notify Playspace User Removed(execute=@n24.then_1, self:Playspace Object Reference=@n21, Delegate=@n12) {is=title K2Node_AddDelegate}
   n25 = For Each Loop(
    Exec=@n23.then_1
    Array:Array of Player State Object References=GetPlayersInBandByController!(
     self:Pilgrim Quickplay Band State Component Object Reference=GetBandState!(ContextActor:Actor Object Reference=@n21) -> (ReturnValue:Pilgrim Quickplay Band State Component Object Reference)
     Controller:Player Controller Object Reference=GetOwningPlayer()
    ) -> (ReturnValue:Array of Player State Object References)
   ) -> (Array Element:Player State Object Reference)
   > AddPlayerReadyIcon(PlayerState=@n25) {is=title K2Node_CallFunction}
   Event Destruct()
   > Parent: Destruct()
   n26 = Get QuickplayStateBeganEventHandle() -> (QuickplayStateBeganEventHandle:Async Action Start Listening To Event Object Reference)
   > Is Valid(InputObject=@n26)
   > StopListeningToEvent(self:Async Action Start Listening To Event Object Reference=@n26)
   > Set QuickplayStateBeganEventHandle()
   Reroute Node(InputPin:Wildcard) -> (OutputPin:Wildcard)
  ]}]
variables [
 Event Router {type=GameplayEventRouterComponent, default=null}
 MaxListItems {type=integer, default=4}
 QuickplayStateBeganEventHandle {type=AsyncAction_StartListeningToEvent, default=null}
]
widgets {
 info {inheritedWidgetCount=0, parentClass=&CommonUserWidget, rootWidgetClass=&Overlay, widgetCount=2}
 widgets [
  Overlay_0 {widgetClassPath=&Overlay, widget=Overlay_0}
  PlayerReadyIconList {
   widgetClassPath=&CommonListView
   bIsVariable=true
   parent=Overlay_0
   widget=PlayerReadyIconList
   properties {
    bIsFocusable=false
    entryWidgetClass=&/WBP_PilgrimQuickplay_AvatarIcon_Ready_ListEntry.WBP_PilgrimQuickplay_AvatarIcon_Ready_ListEntry_C
    horizontalEntrySpacing=8
    numDesignerPreviewEntries=4
    orientation=Orient_Horizontal
    selectionMode=None
    verticalEntrySpacing=8
   }
   slot {class=OverlaySlot}
  }
 ]
}
